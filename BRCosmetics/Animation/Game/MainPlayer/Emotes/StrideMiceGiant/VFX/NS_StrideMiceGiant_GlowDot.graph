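# Niagara system for the StrideMiceGiant emote, drawing a glowing dot grid on the ground around the owner with a tinted light.
# Treadmilled sprites wrap around the owner, fading by circular mask and distance, colored by a distance gradient.
# Dots001 ground-traces dots mainly when they wrap; FE_Train takes socket height, drifting in front end.
# Rail float crossfades the two dot emitters; old-material and debug emitters are disabled.
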NiagaraSystem /BRCosmetics/Animation/Game/MainPlayer/Emotes/StrideMiceGiant/VFX/NS_StrideMiceGiant_GlowDot
dependencies {
 usedDataInterfaces [
  &NiagaraDataInterfaceCurve
  &NiagaraDataInterfaceMeshRendererInfo
  &NiagaraDataInterfaceCollisionQuery
  &NiagaraDataInterfaceDebugDraw
  &NiagaraDataInterfaceArrayFloat
  &NiagaraDataInterfaceColorCurve
  &NiagaraDataInterfaceArrayColor
  &NiagaraDataInterfaceSocketReader
  &NiagaraDataInterfaceArrayFloat3
 ]
 usedDynamicInputs [
  &RandomRangeFloat
  &SelectVectorFromArray
  /Niagara/DynamicInputs/Lerp/SmoothLerpOverTime_Float
 ]
 usedModules [
  &SystemState
  &self:SystemSpawnScript.SystemScriptSource.SystemScriptGraph.NiagaraNodeAssignment_6.SetVariables_#1
  &self:SystemSpawnScript.SystemScriptSource.SystemScriptGraph.NiagaraNodeAssignment_1.SetVariables_#2
  &self:SystemSpawnScript.SystemScriptSource.SystemScriptGraph.NiagaraNodeAssignment_0.SetVariables_#3
  &self:ScratchModule_02
 ]
}
emitters {
 Dots001 {
  emitterSpawnScript [/SetVariables() {enabled=false, isSetParameters=true}]
  emitterUpdateScript [
   &EmitterState(
    Inactive Response=::NewEnumerator1
    Life Cycle Mode=::NewEnumerator1
    Loop Behavior=::NewEnumerator1
    Loop Duration Mode=::NewEnumerator1
    UseLoopDelay=false
   )
   /SetVariables(
    Emitter.OriginOffset:Vector3f=&TransformVector(Vector=<0 0 0>, Destination Space=::World)
    Emitter.PaddingPerCell=5.5
    Emitter.VolumeRadius=220
   ) {isSetParameters=true, module=SetVariables_#4}
   /Niagara/Modules/Spawn/Location/SpawnParticlesInGrid(
    X Count:NiagaraInt32=&MakeIntFromFloat(Float=&FN_FrontEndFloatModifier(Gameplay=12, Front End=0))
    Y Count:NiagaraInt32=&MakeIntFromFloat(Float=&FN_FrontEndFloatModifier(Gameplay=12, Front End=0))
    Z Count=0
   )
   &self:ScratchModule(DistanceMultiplier=1.1, Trace Origin Height=150, TraceDir:Vector3f=<0 0 -1>, TraceDist=300) {module=ScratchModule}
   &DoOnce(
    Reset:NiagaraBool=&SetBoolByFloatComparison(A=$Emitter.Timeline002.Time, B=0.9)
    Trigger Condition:NiagaraBool=&SetBoolByFloatComparison(
     A=&MakeCustomFloatFromBool(Bool=$Emitter.NewElevation, True Float=1, False Float=0)
     B=&MakeCustomFloatFromBool(Bool=$Emitter.PlayerGroundDistanceBool, True Float=1, False Float=0)
     Comparison Type=::NewEnumerator5
    )
   ) {enabled=false, module=DoOnce001}
   &Timeline(Instantaneous Rewind=true, Loop While Playing=true, Play=$Emitter.DoOnce001.TriggeredOnce, Play Rate=5) {enabled=false, module=Timeline002}
   /SetVariables(Emitter.NewElevation:NiagaraBool=$Emitter.PlayerGroundDistanceBool) {enabled=false, isSetParameters=true}
  ]
  particleSpawnScript [
   &FN_SetSpawnGroup(Number of Spawn Groups=6) {module=FN_SetSpawnGroup001}
   &InitializeParticle(
    Position Offset Coordinate Space=::Local
    Sprite Rotation Mode=::NewEnumerator2
    Sprite Size Mode=::NewEnumerator3
    Uniform Sprite Size=34
    UsePositionOffset=true
    Write Material Random=true
   )
   &SocketLocation(
    Orientation Sampling=::NewEnumerator1
    Position Sampling=::NewEnumerator1
    SamplePosOffset=false
    Socket Reader=$System.Socket Reader
    Socket Sampling Mode=::NewEnumerator3
    VelocitySpeedLimit=false
   )
   /Niagara/Modules/Spawn/Location/V2/GridLocation(
    Coordinate Space:ENiagaraCoordinateSpace=::World
    Module.Jitter Inside Grid Cells=false
    Module.Override Random Seed=false
    Normalize Offsets=false
    XYZ Dimensions:Vector3f=&MakeVector(
     X=&Divide_Float(A=$Emitter.VolumeRadius, B=$Emitter.PaddingPerCell)
     Y=&Divide_Float(A=$Emitter.VolumeRadius, B=$Emitter.PaddingPerCell)
     Z=0
    )
   )
   /SetVariables() {isSetParameters=true}
  ]
  particleUpdateScript [
   &ParticleState(Kill Particles When Lifetime Has Elapsed=false, Let Infinitely Lived Particles Die When Emitter Deactivates=true)
   &ScaleSpriteSize(Scale Sprite Size Mode=::NewEnumerator0, Uniform Scale Factor=$System.Rail Float Inverted)
   &SolveForcesAndVelocity(Clamp Velocity=false, Limit Acceleration=false)
   /SetVariables(Particles.PositionCenter:NiagaraPosition=&AddVectorToPosition(Position=$Engine.Owner.Position, Vector=$Emitter.OriginOffset)) {isSetParameters=true, module=SetVariables_#5}
   &self:TreadmillParticles(
    GridSize:NiagaraFloat=&Multiply_Float(A=$Emitter.VolumeRadius, B=2)
    InputPosition:NiagaraPosition=$Particles.PositionCenter
   ) {module=TreadmillParticles}
   &DoOnce(
    Reset:NiagaraBool=&SetBoolByFloatComparison(A=$Particles.Timeline001.Time, B=0.9)
    Trigger Condition:NiagaraBool=&SetBoolByFloatComparison(
     A=&VectorLength(Vector=Linked() {linkedVariable {name=Particles.Position, type=NiagaraPosition}})
     B=&VectorLength(Vector=Linked() {linkedVariable {name=Particles.Previous.Position, type=NiagaraPosition}})
     Comparison Type=::NewEnumerator5
    )
   )
   &Timeline(Instantaneous Rewind=true, Loop While Playing=true, Play=$Particles.DoOnce.TriggeredOnce, Play Rate=10) {module=Timeline001}
   &self:GridCollision(
    CharacterPos:NiagaraPosition=&AddVectorToPosition(Position=$Engine.Owner.Position, Vector=<0 0 50>)
    Disable Trace:NiagaraBool=&Boolean_AndOperation(
     Bool1=&Boolean_NotOperation(Bool=$Particles.DoOnce.TriggeredOnce)
     Bool2=&SetBoolByFloatComparison(A=$Emitter.Age, B=1)
    )
    Height Adjustment:NiagaraFloat=&MakeCustomFloatFromBool(Bool=$Emitter.PlayerGroundDistanceBool, True Float=$Emitter.PlayerGroundDistance, False Float=1.5)
    Trace Origin Height=150
    TraceDir:Vector3f=<0 0 -1>
    TraceDist=250
    bIsFrontEnd:NiagaraBool=$User.bIsFrontEndPreview
   ) {module=GridCollision}
   /SetVariables(
    Particles.Position=/Niagara/DynamicInputs/TypeConversions/MakeCustomPositionFromBool(
     Bool:NiagaraBool=$Particles.DoOnce.TriggeredOnce
     True Position:NiagaraPosition=$Particles.CollisionPos
     False Position:NiagaraPosition=&MakeCustomPositionFromBool(
      Bool:NiagaraBool=&SetBoolByFloatComparison(A=$Emitter.Age, B=1)
      True Position:NiagaraPosition=$Particles.Previous.Position
      False Position:NiagaraPosition=$Particles.CollisionPos
     )
    )
    Particles.SpriteFacing=$Particles.CollisionNormal
   ) {isSetParameters=true}
   /SetVariables(
    Particles.DistanceRange:NiagaraFloat=&NormalizeDistanceRange(
     Start Position:NiagaraPosition=/Niagara/DynamicInputs/Helpers/SimulationPosition()
     End Position:NiagaraPosition=$Particles.Position
     Distance:NiagaraFloat=0
     Invert Normalized Range:NiagaraBool=false
    )
    Particles.DistanceRangeBoost=0.5
   ) {enabled=false, isSetParameters=true}
   /SetVariables(Particles.Position=&MakeVector(
     X=&MakeFloatFromVector(Vector=Linked() {linkedVariable {name=Particles.Position, type=NiagaraPosition}}, Channel=::NewEnumerator0)
     Y=&MakeFloatFromVector(Vector=Linked() {linkedVariable {name=Particles.Position, type=NiagaraPosition}}, Channel=::NewEnumerator1)
     Z=&Add_Float(
      A=&MakeFloatFromVector(Vector=Linked() {linkedVariable {name=Particles.Position, type=NiagaraPosition}}, Channel=::NewEnumerator2)
      B=2
     )
    )) {enabled=false, isSetParameters=true}
   /SetVariables(Particles.Position=&MakeCustomPositionFromBool(
     Bool:NiagaraBool=$Particles.DoOnce.TriggeredOnce
     True Position:NiagaraPosition=&MakeVector(
      X=&MakeFloatFromVector(Vector=Linked() {linkedVariable {name=Particles.CollisionPos, type=NiagaraPosition}}, Channel=::NewEnumerator0)
      Y=&MakeFloatFromVector(Vector=Linked() {linkedVariable {name=Particles.CollisionPos, type=NiagaraPosition}}, Channel=::NewEnumerator1)
      Z=&Lerp_Float(
       A=&MakeFloatFromVector(
        Vector=&AddVectorToPosition(
         Position=&SimulationPosition()
         Vector=&MakeVector(X=0, Y=0, Z=$Particles.DistanceRangeBoost)
        )
        Channel=::NewEnumerator2
       )
       B=&MakeFloatFromVector(Vector=Linked() {linkedVariable {name=Particles.CollisionPos, type=NiagaraPosition}}, Channel=::NewEnumerator2)
       Alpha=$Particles.DistanceRange
      )
     )
     False Position:NiagaraPosition=&MakeCustomPositionFromBool(
      Bool:NiagaraBool=&SetBoolByFloatComparison(A=$Emitter.Age, B=1)
      True Position:NiagaraPosition=&MakeVector(
       X=&MakeFloatFromVector(Vector=Linked() {linkedVariable {name=Particles.Previous.Position, type=NiagaraPosition}}, Channel=::NewEnumerator0)
       Y=&MakeFloatFromVector(Vector=Linked() {linkedVariable {name=Particles.Previous.Position, type=NiagaraPosition}}, Channel=::NewEnumerator1)
       Z=&Lerp_Float(
        A=&MakeFloatFromVector(
         Vector=&AddVectorToPosition(
          Position=&SimulationPosition()
          Vector=&MakeVector(X=0, Y=0, Z=$Particles.DistanceRangeBoost)
         )
         Channel=::NewEnumerator2
        )
        B=&MakeFloatFromVector(Vector=Linked() {linkedVariable {name=Particles.Previous.Position, type=NiagaraPosition}}, Channel=::NewEnumerator2)
        Alpha=$Particles.DistanceRange
       )
      )
      False Position:NiagaraPosition=&MakeVector(
       X=&MakeFloatFromVector(Vector=Linked() {linkedVariable {name=Particles.CollisionPos, type=NiagaraPosition}}, Channel=::NewEnumerator0)
       Y=&MakeFloatFromVector(Vector=Linked() {linkedVariable {name=Particles.CollisionPos, type=NiagaraPosition}}, Channel=::NewEnumerator1)
       Z=&Lerp_Float(
        A=&MakeFloatFromVector(
         Vector=&AddVectorToPosition(
          Position=&SimulationPosition()
          Vector=&MakeVector(X=0, Y=0, Z=$Particles.DistanceRangeBoost)
         )
         Channel=::NewEnumerator2
        )
        B=&MakeFloatFromVector(Vector=Linked() {linkedVariable {name=Particles.CollisionPos, type=NiagaraPosition}}, Channel=::NewEnumerator2)
        Alpha=$Particles.DistanceRange
       )
      )
     )
    )) {enabled=false, isSetParameters=true}
   &RegionMask(
    Invert Region=true
    Offset:Vector3f=$Emitter.OriginOffset
    Region Falloff=0
    Region Radius:NiagaraFloat=&Multiply_Float(A=180, B=1)
    UseOffset=true
    UseRotation=false
   )
   &Timeline(
    Additional Output:ENiagara_TimelineOutput=::NewEnumerator1
    Additional Output Mode:ENiagara_TimelineOutputMode=::NewEnumerator1
    Float Curve=DataInterface() {propertyValues {Curve {keys [{arriveTangent=0.9218, interpMode=RCIM_Cubic, leaveTangent=0.9218, tangentMode=RCTM_User, time=0, value=0}, <at 1 1 RCIM_Cubic>]}}}
    Play=&Boolean_NotOperation(Bool=$Particles.RegionMask.IsInsideMask)
    Play Rate=&MakeCustomFloatFromBool(
     Bool=&SetBoolByFloatComparison(A=&VectorLength(Vector=$User.OwnerLocalVelocity), B=160)
     True Float=10
     False Float=1
    )
    Rewind Rate:NiagaraFloat=&MakeCustomFloatFromBool(
     Bool=&SetBoolByFloatComparison(A=&VectorLength(Vector=$User.OwnerLocalVelocity), B=160)
     True Float=10
     False Float=3
    )
   )
   &Color(
    Color=&ColorFromCurve(
     ColorCurve=DataInterface() {propertyValues {
       AlphaCurve {keys [<at 0 1 RCIM_Cubic>,<at 1 1 RCIM_Cubic>]}
       BlueCurve {keys [
         <at 0.3939 1>
         <at 0.4683 0.1221>
         {arriveTangent=2.6192, interpMode=RCIM_Cubic, leaveTangent=2.6192, time=0.6141, value=0.1094}
         <at 0.7823 0.9447>
         {interpMode=RCIM_Cubic, leaveTangent=0.3125, time=0.9217, value=1}
         <at 1 1>
        ]}
       GreenCurve {keys [
         <at 0.3939 0.1845>
         <at 0.4683 0.3613>
         {arriveTangent=-0.57, interpMode=RCIM_Cubic, leaveTangent=-0.57, time=0.6141, value=0.1094}
         <at 0.7823 0.1823>
         {interpMode=RCIM_Cubic, leaveTangent=0.0123, time=0.9217, value=0.0841}
         <at 1 0.1845>
        ]}
       RedCurve {keys [
         <at 0.3939 0.0356>
         <at 0.4683 1>
         <at 0.6141 1 RCIM_Cubic>
         <at 0.7823 1>
         {interpMode=RCIM_Cubic, leaveTangent=-5.4456, time=0.9217, value=0.3788}
         <at 1 0.0356>
        ]}
      }}
     CurveIndex=&NormalizeDistanceRange(
      Start Position:NiagaraPosition=&SimulationPosition()
      End Position:NiagaraPosition=&AddVectorToPosition(Position=$Particles.Position, Vector=<0 0 90>)
      Distance:NiagaraFloat=&Multiply_Float(A=500, B=$User.Distance)
      Invert Normalized Range:NiagaraBool=false
     )
    )
    Scale Alpha=&FloatFromCurve(
     FloatCurve=DataInterface() {propertyValues {Curve {keys [
         {interpMode=RCIM_Cubic, tangentMode=RCTM_Break, time=0, value=1.36}
         {interpMode=RCIM_Cubic, tangentMode=RCTM_Break, time=0.1331, value=1.0302}
         {interpMode=RCIM_Cubic, tangentMode=RCTM_Break, time=0.2358, value=0.2641}
         {interpMode=RCIM_Cubic, tangentMode=RCTM_Break, time=0.34, value=0}
        ]}}}
     CurveIndex=&NormalizeDistanceRange(
      Start Position:NiagaraPosition=&SimulationPosition()
      End Position:NiagaraPosition=$Particles.Position
      Distance:NiagaraFloat=&Multiply_Float(A=340, B=1.29)
      Invert Normalized Range:NiagaraBool=false
     )
     Scale Curve=$System.Rail Float Inverted
    )
    Scale Color=&VectorFromFloat(Value=&Multiply_Float(A=$Particles.Timeline.Float, B=&Multiply_Float(A=1.5, B=1)))
   ) {module=Color002}
   /SetVariables(
    Particles.GlowColor=&Multiply_LinearColorByFloat(Linear Color=$Particles.Color, Float=1)
    Particles.GlowSize=&Multiply_Vector2D_ByFloat(Float=1.6, Vector2D=$Particles.SpriteSize)
   ) {isSetParameters=true, module=SetVariables_#6}
   &CameraOffset(Camera Offset Amount=&Lerp_Float(
     A=0
     B=50
     Alpha=&NormalizeDistanceRange(
      Start Position:NiagaraPosition=$Engine.Owner.Position
      End Position:NiagaraPosition=$Particles.Position
      Distance:NiagaraFloat=&Multiply_Float(A=200, B=1)
      Invert Normalized Range:NiagaraBool=false
     )
    )) {enabled=false}
  ]
  rendererClasses [&NiagaraSpriteRendererProperties]
  renderers [
   {rendererClass=&NiagaraSpriteRendererProperties}
   {rendererClass=&NiagaraSpriteRendererProperties, rendererIndex=1}
  ]
 }
 FE_Train {
  emitterUpdateScript [
   &EmitterState(
    Inactive Response=::NewEnumerator1
    Life Cycle Mode=::NewEnumerator1
    Loop Behavior=::NewEnumerator1
    Loop Duration Mode=::NewEnumerator1
    UseLoopDelay=false
   )
   /SetVariables(
    Emitter.OriginOffset:Vector3f=&TransformVector(Vector=<0 0 0>, Destination Space=::World)
    Emitter.PaddingPerCell=5.5
    Emitter.VolumeRadius=220
   ) {isSetParameters=true, module=SetVariables_#4}
   &SpawnParticlesInGrid(X Count=12, Y Count=12, Z Count=0)
  ]
  particleSpawnScript [
   &FN_SetSpawnGroup(Number of Spawn Groups=6) {module=FN_SetSpawnGroup001}
   &InitializeParticle(
    Position Offset=<0 0 2>
    Position Offset Coordinate Space=::Local
    Sprite Rotation Mode=::NewEnumerator2
    Sprite Size Mode=::NewEnumerator3
    Uniform Sprite Size=34
    UsePositionOffset=true
    Write Material Random=true
   )
   &GridLocation(
    Coordinate Space:ENiagaraCoordinateSpace=::World
    Module.Jitter Inside Grid Cells=false
    Module.Override Random Seed=false
    Normalize Offsets=false
    XYZ Dimensions:Vector3f=&MakeVector(
     X=&Divide_Float(A=$Emitter.VolumeRadius, B=$Emitter.PaddingPerCell)
     Y=&Divide_Float(A=$Emitter.VolumeRadius, B=$Emitter.PaddingPerCell)
     Z=0
    )
   )
  ]
  particleUpdateScript [
   &ParticleState(Kill Particles When Lifetime Has Elapsed=false, Let Infinitely Lived Particles Die When Emitter Deactivates=true)
   &ScaleSpriteSize(
    Scale Sprite Size Mode=::NewEnumerator0
    Uniform Scale Factor=&Add_Float(
     A=$System.Rail Float
     B=&MakeCustomFloatFromBool(Bool=&Boolean_OrOperation(Bool1=$User.bIsFrontEnd, Bool2=$User.bIsFrontEndPreview), True Float=1, False Float=0)
    )
   )
   &SocketLocation(
    Orientation Sampling=::NewEnumerator1
    Position Sampling=::NewEnumerator1
    SamplePosOffset=true
    Sampled Position Offset=&MakeVector(
     X=0
     Y=0
     Z=&MakeCustomFloatFromBool(Bool=&Boolean_OrOperation(Bool1=$User.bIsFrontEnd, Bool2=$User.bIsFrontEndPreview), True Float=0.4, False Float=5)
    )
    Socket Reader=$System.Socket Reader
    Socket Sampling Mode=::NewEnumerator3
    VelocitySpeedLimit=false
   )
   &SolveForcesAndVelocity(Clamp Velocity=false, Limit Acceleration=false)
   /SetVariables(Particles.PositionCenter:NiagaraPosition=&AddVectorToPosition(Position=$Engine.Owner.Position, Vector=$Emitter.OriginOffset)) {isSetParameters=true, module=SetVariables_#5}
   &self:TreadmillParticles(
    GridSize:NiagaraFloat=&Multiply_Float(A=$Emitter.VolumeRadius, B=2)
    InputPosition:NiagaraPosition=$Particles.PositionCenter
   ) {module=TreadmillParticles}
   /SetVariables(
    Particles.Position=&AddVectorToPosition(
     Position=&MakeVector(
      X=&MakeFloatFromVector(Vector=Linked() {linkedVariable {name=Particles.Position, type=NiagaraPosition}}, Channel=::NewEnumerator0)
      Y=&MakeFloatFromVector(Vector=Linked() {linkedVariable {name=Particles.Position, type=NiagaraPosition}}, Channel=::NewEnumerator1)
      Z=&MakeFloatFromVector(Vector=Linked() {linkedVariable {name=Output.SocketLocation.SampledPosition, type=NiagaraPosition}}, Channel=::NewEnumerator2)
     )
     Vector=&FN_FrontEndVector3Modifier(
      Gameplay=<0 0 0>
      Front End=&RotateVectorByQuaternion(
       Vector To Rotate=&Multiply_VectorByFloat(Vector=<0 -50 0>, Float=$Engine.DeltaTime)
       Quaternion=$Output.SocketLocation.SampledRotation
      )
     )
    )
    Particles.SpriteFacing=<0 0 1>
   ) {isSetParameters=true, module=SetVariables_#7}
   &RegionMask(
    Invert Region=true
    Offset:Vector3f=$Emitter.OriginOffset
    Region Falloff=0
    Region Radius:NiagaraFloat=&Multiply_Float(A=180, B=1)
    UseOffset=true
    UseRotation=false
   )
   &Timeline(
    Additional Output:ENiagara_TimelineOutput=::NewEnumerator1
    Additional Output Mode:ENiagara_TimelineOutputMode=::NewEnumerator1
    Float Curve=DataInterface() {propertyValues {Curve {keys [{arriveTangent=0.9218, interpMode=RCIM_Cubic, leaveTangent=0.9218, tangentMode=RCTM_User, time=0, value=0}, <at 1 1 RCIM_Cubic>]}}}
    Play=&Boolean_NotOperation(Bool=$Particles.RegionMask.IsInsideMask)
    Play Rate=&MakeCustomFloatFromBool(
     Bool=&SetBoolByFloatComparison(A=&VectorLength(Vector=$User.OwnerLocalVelocity), B=160)
     True Float=10
     False Float=1
    )
    Rewind Rate:NiagaraFloat=&MakeCustomFloatFromBool(
     Bool=&SetBoolByFloatComparison(A=&VectorLength(Vector=$User.OwnerLocalVelocity), B=160)
     True Float=10
     False Float=3
    )
   )
   &Color(
    Color=&ColorFromCurve(
     ColorCurve=DataInterface() {propertyValues {
       AlphaCurve {keys [<at 0 1 RCIM_Cubic>,<at 1 1 RCIM_Cubic>]}
       BlueCurve {keys [
         <at 0.3939 1>
         <at 0.4683 0.1221>
         {arriveTangent=2.6192, interpMode=RCIM_Cubic, leaveTangent=2.6192, time=0.6141, value=0.1094}
         <at 0.7823 0.9447>
         {arriveTangent=0.2542, interpMode=RCIM_Cubic, leaveTangent=0.2542, time=0.9217, value=1}
         <at 1 1>
        ]}
       GreenCurve {keys [
         <at 0.3939 0.1845>
         <at 0.4683 0.3613>
         {arriveTangent=-0.57, interpMode=RCIM_Cubic, leaveTangent=-0.57, time=0.6141, value=0.1094}
         <at 0.7823 0.1823>
         {arriveTangent=0.01, interpMode=RCIM_Cubic, leaveTangent=0.01, time=0.9217, value=0.0841}
         <at 1 0.1845>
        ]}
       RedCurve {keys [
         <at 0.3939 0.0356>
         <at 0.4683 1>
         <at 0.6141 1 RCIM_Cubic>
         <at 0.7823 1>
         {arriveTangent=-4.4297, interpMode=RCIM_Cubic, leaveTangent=-4.4297, time=0.9217, value=0.3788}
         <at 1 0.0356>
        ]}
      }}
     CurveIndex=&NormalizeDistanceRange(
      Start Position:NiagaraPosition=&SimulationPosition()
      End Position:NiagaraPosition=&AddVectorToPosition(Position=$Particles.Position, Vector=<0 0 90>)
      Distance:NiagaraFloat=&Multiply_Float(A=500, B=$User.Distance)
      Invert Normalized Range:NiagaraBool=false
     )
    )
    Scale Alpha=&FloatFromCurve(
     FloatCurve=DataInterface() {propertyValues {Curve {keys [
         {interpMode=RCIM_Cubic, tangentMode=RCTM_Break, time=0, value=1.36}
         {interpMode=RCIM_Cubic, tangentMode=RCTM_Break, time=0.1331, value=1.0302}
         {interpMode=RCIM_Cubic, tangentMode=RCTM_Break, time=0.2358, value=0.2641}
         {interpMode=RCIM_Cubic, tangentMode=RCTM_Break, time=0.34, value=0}
        ]}}}
     CurveIndex=&NormalizeDistanceRange(
      Start Position:NiagaraPosition=&SimulationPosition()
      End Position:NiagaraPosition=$Particles.Position
      Distance:NiagaraFloat=&Multiply_Float(A=340, B=1.29)
      Invert Normalized Range:NiagaraBool=false
     )
     Scale Curve=&Add_Float(
      A=$System.Rail Float
      B=&MakeCustomFloatFromBool(Bool=&Boolean_OrOperation(Bool1=$User.bIsFrontEnd, Bool2=$User.bIsFrontEndPreview), True Float=1, False Float=0)
     )
    )
    Scale Color=&VectorFromFloat(Value=&Multiply_Float(A=$Particles.Timeline.Float, B=&Multiply_Float(A=1.5, B=1)))
   ) {module=Color002}
   /SetVariables(
    Particles.GlowColor=&Multiply_LinearColorByFloat(Linear Color=$Particles.Color, Float=1)
    Particles.GlowSize=&Multiply_Vector2D_ByFloat(Float=1.6, Vector2D=$Particles.SpriteSize)
   ) {isSetParameters=true, module=SetVariables_#6}
  ]
  rendererClasses [&NiagaraSpriteRendererProperties]
  renderers [
   {rendererClass=&NiagaraSpriteRendererProperties}
   {rendererClass=&NiagaraSpriteRendererProperties, rendererIndex=1}
  ]
 }
 FN_Debug {
  bEnabled=false
  emitterUpdateScript [
   &EmitterState()
   &SpawnBurst_Instantaneous(Spawn Count=1, Use Spawn Probability=false)
  ]
  particleSpawnScript [/Niagara/Modules/Spawn/Initialization/InitializeParticle(
    Lifetime=5
    Module.Write Color=true
    Module.Write Lifetime=true
    Module.Write Mass=true
    Module.Write Material Random=true
    Module.Write Position=false
    Module.Write Scale=false
    Module.Write SpriteRotation=false
    Module.Write SpriteSize=true
    Position=/Niagara/Functions/Localspace/SimulationPosition()
    Sprite Size=<100 100>
   )]
  particleUpdateScript [
   &ParticleState()
   /SetVariables(Particles.Debug:NiagaraFloat=&NormalizeDistanceRange(
     Start Position:NiagaraPosition=/Niagara/DynamicInputs/Helpers/SimulationPosition()
     End Position:NiagaraPosition=$Particles.Position
     Distance:NiagaraFloat=&Multiply_Float(A=500, B=$User.Distance)
     Invert Normalized Range:NiagaraBool=false
    )) {isSetParameters=true}
   &DynamicMaterialParameters(
    Index 0 Param 1=$Particles.Debug
    Module.Param0WriteEnabled=true
    Module.Param1WriteEnabled=true
    Module.Param2WriteEnabled=true
    Module.Param3WriteEnabled=true
   )
  ]
  rendererClasses [&NiagaraSpriteRendererProperties]
  renderers [{rendererClass=&NiagaraSpriteRendererProperties}]
 }
 Light {
  emitterUpdateScript [
   &EmitterState(Life Cycle Mode=::NewEnumerator1, Loop Behavior=::NewEnumerator1, Loop Duration Mode=::NewEnumerator1, UseLoopDelay=false)
   &SpawnBurst_Instantaneous(Use Spawn Probability=false, UseLoopCountLimit=false)
  ]
  particleSpawnScript [
   /Niagara/Modules/Spawn/Initialization/V2/InitializeParticle(Sprite Size Mode=::NewEnumerator3, Uniform Sprite Size=30, UsePositionOffset=false, Write Material Random=true)
   /SetVariables(Particles.LightExponent=2, Particles.LightRadius=80) {isSetParameters=true}
  ]
  particleUpdateScript [
   &ParticleState(Kill Particles When Lifetime Has Elapsed=false, Let Infinitely Lived Particles Die When Emitter Deactivates=true)
   &ScaleColor(
    Scale RGB=&Multiply_Vector(
     A=&MakeVectorFromLinearColorRGB(Color=&SelectLinearColorFromArray(
       Array Sampling Mode=::NewEnumerator1
       Direct Array Index=&MakeIntFromFloat(Float=$User.BurstIndex, Conversion Mode=::NewEnumerator1)
       Color Selection Array=DataInterface() {propertyValues {ColorData [<rgba 0 6.6126 35 1>, <rgba 35 3.8312 3.8312 1>, <rgba 35 6.3407 33.1557 1>]}}
       Random Seed=?
      ))
     B=&VectorFromFloat(Value=&FNTimeOfDayScalar(Night Scale=1, Morning Scale=1, Day Scale=1.2, Evening Scale=1))
    )
    ScaleA=true
    ScaleRGB=true
   )
  ]
  rendererClasses [&NiagaraLightRendererProperties]
  renderers [{rendererClass=&NiagaraLightRendererProperties, rendererIndex=0}]
 }
 OldMaterial {
  bEnabled=false
  emitterUpdateScript [
   &EmitterState(
    Inactive Response=::NewEnumerator1
    Life Cycle Mode=::NewEnumerator1
    Loop Behavior=::NewEnumerator1
    Loop Duration Mode=::NewEnumerator1
    UseLoopDelay=false
   )
   /SetVariables(
    Emitter.OriginOffset:Vector3f=&TransformVector(Vector=<0 0 0>, Destination Space=::World)
    Emitter.PaddingPerCell=5.5
    Emitter.VolumeRadius=220
   ) {isSetParameters=true, module=SetVariables_#4}
   &SpawnParticlesInGrid(X Count=12, Y Count=12, Z Count=0)
  ]
  particleSpawnScript [
   &FN_SetSpawnGroup(Number of Spawn Groups=6) {module=FN_SetSpawnGroup001}
   &InitializeParticle(
    Position Offset Coordinate Space=::Local
    Sprite Rotation Mode=::NewEnumerator2
    Sprite Size Mode=::NewEnumerator3
    Uniform Sprite Size=34
    UsePositionOffset=true
    Write Material Random=true
   )
   &GridLocation(
    Coordinate Space:ENiagaraCoordinateSpace=::World
    Module.Jitter Inside Grid Cells=false
    Module.Override Random Seed=false
    Normalize Offsets=false
    XYZ Dimensions:Vector3f=&MakeVector(
     X=&Divide_Float(A=$Emitter.VolumeRadius, B=$Emitter.PaddingPerCell)
     Y=&Divide_Float(A=$Emitter.VolumeRadius, B=$Emitter.PaddingPerCell)
     Z=0
    )
   )
  ]
  particleUpdateScript [
   &ParticleState(Kill Particles When Lifetime Has Elapsed=false, Let Infinitely Lived Particles Die When Emitter Deactivates=true)
   &SocketLocation(
    Orientation Sampling=::NewEnumerator1
    Position Sampling=::NewEnumerator1
    SamplePosOffset=false
    Socket Reader=$System.Socket Reader
    Socket Sampling Mode=::NewEnumerator3
    VelocitySpeedLimit=false
   )
   &SolveForcesAndVelocity(Clamp Velocity=false, Limit Acceleration=false)
   /SetVariables(Particles.PositionCenter:NiagaraPosition=&AddVectorToPosition(Position=$Engine.Owner.Position, Vector=$Emitter.OriginOffset)) {isSetParameters=true, module=SetVariables_#5}
   &self:TreadmillParticles(
    GridSize:NiagaraFloat=&Multiply_Float(A=$Emitter.VolumeRadius, B=2)
    InputPosition:NiagaraPosition=$Particles.PositionCenter
   ) {module=TreadmillParticles}
   /SetVariables(Particles.Position=&MakeVector(
     X=&MakeFloatFromVector(Vector=Linked() {linkedVariable {name=Particles.Position, type=NiagaraPosition}}, Channel=::NewEnumerator0)
     Y=&MakeFloatFromVector(Vector=Linked() {linkedVariable {name=Particles.Position, type=NiagaraPosition}}, Channel=::NewEnumerator1)
     Z=&MakeFloatFromVector(Vector=&SimulationPosition(), Channel=::NewEnumerator2)
    )) {isSetParameters=true}
   /SetVariables(Particles.Position=&AddVectorToPosition(
     Position=$Particles.Position
     Vector=&MaskVector(
      Mask Source=1
      Mask Parameter=&MakeCustomIntFromBool(
       Bool=&SetBoolByFloatComparison(
        A=&Add_Float(
         A=&Add_Float(
          A=&Abs_Float(Float:NiagaraFloat=&MakeFloatFromVector(Vector=$Engine.Owner.Velocity, Channel=::NewEnumerator0))
          B=&Abs_Float(Float:NiagaraFloat=&MakeFloatFromVector(Vector=$Engine.Owner.Velocity, Channel=::NewEnumerator2))
         )
         B=&Abs_Float(Float:NiagaraFloat=&MakeFloatFromVector(Vector=$Engine.Owner.Velocity, Channel=::NewEnumerator1))
        )
        B=300
       )
       True INT=1
       False INT=0
      )
      Mask=0
      Masked Vector=<0 0 90>
      Passthrough Vector=<0 0 0>
     )
    )) {isSetParameters=true}
   /SetVariables(Particles.Position=&MakeCustomVectorFromBool(
     Bool=&SetBoolByFloatComparison(A=&VectorLength(Vector=$User.OwnerLocalVelocity), B=100)
     True Vector=&AddVectorToPosition(Position=$Particles.Position, Vector=<0 0 30>)
     False Vector=&AddVectorToPosition(Position=$Particles.Position, Vector=<0 0 90>)
    )) {enabled=false, isSetParameters=true}
   /SetVariables(Particles.Position=&AddVectorToPosition(
     Position=$Particles.Position
     Vector=&FN_FrontEndVector3Modifier(
      Gameplay=<0 0 0>
      Front End=&RotateVectorByQuaternion(Vector To Rotate=<0 -1.2 0>, Quaternion=$Output.SocketLocation.SampledRotation)
     )
    )) {isSetParameters=true, module=SetVariables_#7}
   &self:GridCollision(TraceDir:Vector3f=<0 0 -1>, TraceDist=250, bIsFrontEnd:NiagaraBool=$User.bIsFrontEndPreview) {module=GridCollision}
   &RegionMask(
    Invert Region=true
    Offset:Vector3f=$Emitter.OriginOffset
    Region Falloff=0
    Region Radius:NiagaraFloat=&Multiply_Float(A=220, B=1)
    UseOffset=true
    UseRotation=false
   )
   &Timeline(
    Additional Output:ENiagara_TimelineOutput=::NewEnumerator1
    Additional Output Mode:ENiagara_TimelineOutputMode=::NewEnumerator1
    Float Curve=DataInterface() {propertyValues {Curve {keys [{arriveTangent=0.9218, interpMode=RCIM_Cubic, leaveTangent=0.9218, tangentMode=RCTM_User, time=0, value=0}, <at 1 1 RCIM_Cubic>]}}}
    Play=&Boolean_NotOperation(Bool=$Particles.RegionMask.IsInsideMask)
    Play Rate=&MakeCustomFloatFromBool(
     Bool=&SetBoolByFloatComparison(A=&VectorLength(Vector=$User.OwnerLocalVelocity), B=160)
     True Float=10
     False Float=1
    )
    Rewind Rate:NiagaraFloat=&MakeCustomFloatFromBool(
     Bool=&SetBoolByFloatComparison(A=&VectorLength(Vector=$User.OwnerLocalVelocity), B=160)
     True Float=10
     False Float=3
    )
   )
   &Color(
    Color=&ColorFromCurve(
     ColorCurve=DataInterface() {propertyValues {
       AlphaCurve {keys [<at 0 1 RCIM_Cubic>,<at 1 1 RCIM_Cubic>]}
       BlueCurve {keys [
         <at 0.3939 1>
         <at 0.4683 0.1221>
         {arriveTangent=2.6192, interpMode=RCIM_Cubic, leaveTangent=2.6192, time=0.6141, value=0.1094}
         <at 0.7823 0.9447>
         {interpMode=RCIM_Cubic, leaveTangent=0.3125, time=0.9217, value=1}
         <at 1 1>
        ]}
       GreenCurve {keys [
         <at 0.3939 0.1845>
         <at 0.4683 0.3613>
         {arriveTangent=-0.57, interpMode=RCIM_Cubic, leaveTangent=-0.57, time=0.6141, value=0.1094}
         <at 0.7823 0.1823>
         {interpMode=RCIM_Cubic, leaveTangent=0.0123, time=0.9217, value=0.0841}
         <at 1 0.1845>
        ]}
       RedCurve {keys [
         <at 0.3939 0.0356>
         <at 0.4683 1>
         <at 0.6141 1 RCIM_Cubic>
         <at 0.7823 1>
         {interpMode=RCIM_Cubic, leaveTangent=-5.4456, time=0.9217, value=0.3788}
         <at 1 0.0356>
        ]}
      }}
     CurveIndex=&NormalizeDistanceRange(
      Start Position:NiagaraPosition=&SimulationPosition()
      End Position:NiagaraPosition=$Particles.Position
      Distance:NiagaraFloat=&Multiply_Float(A=500, B=$User.Distance)
      Invert Normalized Range:NiagaraBool=false
     )
    )
    Scale Alpha=&FloatFromCurve(
     FloatCurve=DataInterface() {propertyValues {Curve {keys [
         {interpMode=RCIM_Cubic, tangentMode=RCTM_Break, time=0, value=1.36}
         {interpMode=RCIM_Cubic, tangentMode=RCTM_Break, time=0.1331, value=1.0302}
         {interpMode=RCIM_Cubic, tangentMode=RCTM_Break, time=0.2358, value=0.2641}
         {interpMode=RCIM_Cubic, tangentMode=RCTM_Break, time=0.34, value=0}
        ]}}}
     CurveIndex=&NormalizeDistanceRange(
      Start Position:NiagaraPosition=&SimulationPosition()
      End Position:NiagaraPosition=$Particles.Position
      Distance:NiagaraFloat=&Multiply_Float(A=500, B=1.29)
      Invert Normalized Range:NiagaraBool=false
     )
    )
    Scale Color=&VectorFromFloat(Value=&Multiply_Float(A=$Particles.Timeline.Float, B=&Multiply_Float(A=1.5, B=1)))
   ) {module=Color002}
   &SpriteFacingAndAlignment(Sprite Alignment=<1 0 0>, Sprite Facing=$Engine.Owner.SystemZAxis, WriteAlignment=true, WriteFacing=true) {enabled=false}
   &SpriteFacingAndAlignment(
    Sprite Alignment=<1 0 0>
    Sprite Facing=&MakeCustomVectorFromBool(
     Bool=&SetBoolByFloatComparison(A=&VectorLength(Vector=$User.OwnerLocalVelocity), B=200)
     True Vector=<0 0 1>
     False Vector=$Engine.Owner.SystemZAxis
    )
    WriteAlignment=true
    WriteFacing=true
   ) {module=SpriteFacingAndAlignment001}
   &CameraOffset(Camera Offset Amount=5)
   /SetVariables(
    Particles.GlowColor=&Multiply_LinearColorByFloat(Linear Color=$Particles.Color, Float=1)
    Particles.GlowSize=&Multiply_Vector2D_ByFloat(Float=1.6, Vector2D=$Particles.SpriteSize)
   ) {isSetParameters=true, module=SetVariables_#6}
  ]
  rendererClasses [&NiagaraSpriteRendererProperties]
  renderers [
   {rendererClass=&NiagaraSpriteRendererProperties}
   {rendererClass=&NiagaraSpriteRendererProperties, rendererIndex=1}
  ]
 }
}
summary {
 systemName=NS_StrideMiceGiant_GlowDot
 emitters [
  {bEnabled=false, emitterName=OldMaterial, rendererClasses [&NiagaraSpriteRendererProperties]}
  {emitterName=FE_Train, rendererClasses [&NiagaraSpriteRendererProperties]}
  {emitterName=Dots001, rendererClasses [&NiagaraSpriteRendererProperties]}
  {emitterName=Light, rendererClasses [&NiagaraLightRendererProperties]}
  {bEnabled=false, emitterName=FN_Debug, rendererClasses [&NiagaraSpriteRendererProperties]}
 ]
 userVariables [
  User.bIsFrontEnd {type=NiagaraBool}
  User.bIsFrontEndPreview {type=NiagaraBool}
  User.BurstIndex {type=NiagaraFloat, defaultValue=0, description=""}
  User.Distance {type=NiagaraFloat, defaultValue=1, description=""}
  User.OwnerLocalVelocity {type=Vector3f, description="", defaultValue {struct=Vector3f, value=<1 1 1>}}
 ]
}
